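# Unlit translucent sprite material for particle and Niagara effects, drawing a directional flare from T_FlareDirectional_01.
# Particle color RGB feeds the MF_FX_EmissiveAdjustments function for the glow color.
# The texture's red channel plus its blue channel, scaled by the Core Glow dynamic parameter, is multiplied by particle alpha to form the flare mask.
# Core Glow defaults to zero, so the blue-channel contribution appears only when that parameter is raised.
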Material /Game/Effects/Fort_Effects/Textures/Explo/M_FlareDirectional_01
graph {
 properties {
  bUsedWithNiagaraSprites=true
  bUsedWithParticleSprites=true
  bUsedWithStaticMesh=false
  blendMode=BLEND_Translucent
  pixelDepthOffsetMode=PDOM_Legacy
  shadingModel=MSM_Unlit
  usedShadingModels=Unlit
 }
 nodes [
  n1 = ParticleColor()
  MaterialFunctionCall(Emissive=@n1.RGB) {materialFunction=&MF_FX_EmissiveAdjustments}
  n2 = TextureSample() {texture=&/T_FlareDirectional_01}
  n3 = DynamicParameter() {defaultValue=<rgba 0 0 0 1>, paramNames [Core Glow, Param2, Param3, Param4]}
  Multiply(A=@n1.A, B=Add(A=@n2.R, B=Multiply(A=@n2.B, B=@n3.Core Glow)))
 ]
}
