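# Two-sided unlit translucent material for a flash ring effect on mesh particles, including Niagara, with particle color fed through an emissive adjustment function.
# A cropped, clamped ring texture forms the mask, optionally multiplied by particle alpha through a static switch.
# A tiled, panning tornado pattern breaks up the ring, with dynamic parameters driving its blend amount, contrast exponent and UV offset.
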Material /Game/Effects/Fort_Effects/Textures/Explo/M_Flash_Ring
graph {
 properties {
  bUsedWithMeshParticles=true
  bUsedWithNiagaraMeshParticles=true
  blendMode=BLEND_Translucent
  pixelDepthOffsetMode=PDOM_Legacy
  shadingModel=MSM_Unlit
  twoSided=true
  usedShadingModels=Unlit
 }
 nodes [
  n1 = ParticleColor()
  MaterialFunctionCall(Emissive=@n1.RGB) {materialFunction=&MF_FX_EmissiveAdjustments}
  n2 = VectorParameter() {parameterName=CropUpperLeft}
  n3 = VectorParameter() {parameterName=CropUpperLeft_1, defaultValue=<rgba 1 0.125 0 0>}
  n4 = TextureSample(UVs=MaterialFunctionCall!(
    UVs=TextureCoordinate()
    TextureIn
    UpperLeftCorner=ComponentMask(None=@n2.RGB) {g=true, r=true}
    LowerRightCorner=ComponentMask(None=@n3.RGB) {g=true, r=true}
    ExportFloat4
   ) {materialFunction=/Engine/Functions/Engine_MaterialFunctions01/Texturing/TextureCropping}.Crop UVs) {samplerSource=SSM_Clamp_WorldGroupSettings, texture=&/T_Flash_Ring}
  n5 = DynamicParameter() {defaultValue=<rgba 0 0 0 1>}
  n6 = Add(A=TextureCoordinate() {uTiling=2, vTiling=20}, B=AppendVector(A=@n5.Param4, B=@n5.Param4))
  n7 = Power(
   Base=TextureSample(UVs=Panner(Coordinate=@n6) {speedX=1, speedY=2}) {texture=/Game/Global/Textures/Default/Tileable_Patterns/T_Tornado_Tile}.RGB
   Exp=@n5.Param2
  )
  Multiply(
   A=LinearInterpolate(A=@n4.RGB, B=Multiply(A=@n4.RGB, B=@n7), Alpha=@n5.Param1)
   B=StaticSwitchParameter(True=@n1.A, False=Constant() {r=1}) {parameterName=UseParticleAlpha}
  )
 ]
}
