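# Unlit, two-sided translucent particle material for sprite and mesh spike effects, using T_Spike_01 and fading out toward the bottom.
# Particle color feeds an emissive-adjustment function and also multiplies the desaturated spike texture.
# An alpha-style product combines particle alpha, texture RGB and a clamped gradient of world Z above the actor's origin.
# That height gradient drops to zero at and below the origin, so the fade follows actor position rather than UVs.
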Material /Game/Effects/Fort_Effects/Textures/Explo/M_Spike_01_Fade_Bottom
graph {
 properties {
  bUsedWithMeshParticles=true
  bUsedWithParticleSprites=true
  blendMode=BLEND_Translucent
  pixelDepthOffsetMode=PDOM_Legacy
  shadingModel=MSM_Unlit
  twoSided=true
  usedShadingModels=Unlit
 }
 nodes [
  Constant()
  n1 = ParticleColor()
  MaterialFunctionCall(Emissive=@n1.RGB) {materialFunction=&MF_FX_EmissiveAdjustments}
  n2 = TextureSample() {texture=&/T_Spike_01}
  Multiply(A=@n1.RGB, B=Desaturation(None=@n2.RGB))
  n3 = WorldPosition()
  n4 = ActorPositionWS()
  n5 = Multiply(A=Subtract(A=ComponentMask(None=@n3.XYZ) {b=true}, B=ComponentMask(None=@n4) {b=true}), B=Constant() {r=0.02})
  Multiply(A=@n1.A, B=Multiply(A=@n2.RGB, B=Clamp(None=@n5)))
 ]
}
